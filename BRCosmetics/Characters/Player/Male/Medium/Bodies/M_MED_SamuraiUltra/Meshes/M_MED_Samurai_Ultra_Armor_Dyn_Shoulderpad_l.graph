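# Physics asset for the Samurai Ultra male medium body's dynamic shoulder pads, covering both sides despite the _l suffix.
# Capsule bodies cover upper spine, clavicles, upper arms and each pad's cnst, master and two base bones.
# Locked angular constraints join spine, clavicles, upper arms and cnst bones into a rigid frame.
# From each cnst, a master, base_01, base_02 chain has tight angular limits, widest on swing2, for restrained secondary motion.
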PhysicsAsset /BRCosmetics/Characters/Player/Male/Medium/Bodies/M_MED_SamuraiUltra/Meshes/M_MED_Samurai_Ultra_Armor_Dyn_Shoulderpad_l
bodies {
 clavicle_l [{shapeType=Capsule, length=3, radius=3}]
 clavicle_r [{shapeType=Capsule, length=3, radius=3}]
 dyn_shoulderArmor_base_l_01 [{shapeType=Capsule, length=6.9375, radius=2.0625, rotation=<rot 0 0 89.9995>}]
 dyn_shoulderArmor_base_l_02 [{shapeType=Capsule, length=6.125, radius=2.875, rotation=<rot 0 0 89.9995>}]
 dyn_shoulderArmor_base_r_01 [{shapeType=Capsule, length=6.9, radius=2.06, rotation=<rot 0 0 90>}]
 dyn_shoulderArmor_base_r_02 [{shapeType=Capsule, length=6.125, radius=2.9, rotation=<rot 0 0 90>}]
 dyn_shoulderArmor_cnst_l [{shapeType=Capsule, length=0.505, radius=0.505, rotation=<rot 0 -90 0>}]
 dyn_shoulderArmor_cnst_r [{shapeType=Capsule, length=0.505, radius=0.505, rotation=<rot 0 -90 0>}]
 dyn_shoulderArmor_master_l [{shapeType=Capsule, length=3, radius=3}]
 dyn_shoulderArmor_master_r [{shapeType=Capsule, length=3, radius=3}]
 spine_05 [{shapeType=Capsule, length=14.6819, radius=13.444, center=<-0.2908 -0.5109 0.0039>, rotation=<rot 0.4162 -90 -0.1594>}]
 upperarm_l [{shapeType=Capsule, length=18, radius=6.25, center=<11.9103 0.0003 0.4111>, rotation=<rot 84.9908 -180 180>}]
 upperarm_r [{shapeType=Capsule, length=18, radius=6.25, center=<-11.9103 0 -0.4111>, rotation=<rot 85 -180 180>}]
}
constraints [
 {
  bone1Name=dyn_shoulderArmor_base_l_01
  bone2Name=dyn_shoulderArmor_master_l
  swing1LimitDegrees=2
  swing2LimitDegrees=15
  twistLimitDegrees=2
 }
 {
  bone1Name=dyn_shoulderArmor_base_l_02
  bone2Name=dyn_shoulderArmor_base_l_01
  swing1LimitDegrees=2
  swing2LimitDegrees=9
  twistLimitDegrees=2
 }
 {bone1Name=dyn_shoulderArmor_master_l, bone2Name=dyn_shoulderArmor_cnst_l, swing1LimitDegrees=3, swing2LimitDegrees=5, twistLimitDegrees=3}
 {
  bone1Name=clavicle_l
  bone2Name=spine_05
  swing1LimitDegrees=45
  swing1Motion=Locked
  swing2LimitDegrees=45
  swing2Motion=Locked
  twistLimitDegrees=45
  twistMotion=Locked
 }
 {
  bone1Name=upperarm_l
  bone2Name=clavicle_l
  swing1LimitDegrees=45
  swing1Motion=Locked
  swing2LimitDegrees=45
  swing2Motion=Locked
  twistLimitDegrees=45
  twistMotion=Locked
 }
 {
  bone1Name=dyn_shoulderArmor_cnst_l
  bone2Name=clavicle_l
  swing1LimitDegrees=45
  swing1Motion=Locked
  swing2LimitDegrees=45
  swing2Motion=Locked
  twistLimitDegrees=45
  twistMotion=Locked
 }
 {
  bone1Name=clavicle_r
  bone2Name=spine_05
  swing1LimitDegrees=45
  swing1Motion=Locked
  swing2LimitDegrees=45
  swing2Motion=Locked
  twistLimitDegrees=45
  twistMotion=Locked
 }
 {
  bone1Name=upperarm_r
  bone2Name=clavicle_r
  swing1LimitDegrees=45
  swing1Motion=Locked
  swing2LimitDegrees=45
  swing2Motion=Locked
  twistLimitDegrees=45
  twistMotion=Locked
 }
 {
  bone1Name=dyn_shoulderArmor_cnst_r
  bone2Name=clavicle_r
  swing1LimitDegrees=45
  swing1Motion=Locked
  swing2LimitDegrees=45
  swing2Motion=Locked
  twistLimitDegrees=45
  twistMotion=Locked
 }
 {bone1Name=dyn_shoulderArmor_master_r, bone2Name=dyn_shoulderArmor_cnst_r, swing1LimitDegrees=3, swing2LimitDegrees=5, twistLimitDegrees=3}
 {
  bone1Name=dyn_shoulderArmor_base_r_01
  bone2Name=dyn_shoulderArmor_master_r
  swing1LimitDegrees=2
  swing2LimitDegrees=15
  twistLimitDegrees=2
 }
 {
  bone1Name=dyn_shoulderArmor_base_r_02
  bone2Name=dyn_shoulderArmor_base_r_01
  swing1LimitDegrees=2
  swing2LimitDegrees=9
  twistLimitDegrees=2
 }
]
