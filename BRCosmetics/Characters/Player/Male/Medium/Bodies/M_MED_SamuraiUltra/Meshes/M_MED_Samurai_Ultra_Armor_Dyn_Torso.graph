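# Physics asset for the Samurai Ultra medium male torso armor, providing simulation bodies for its dangling back tassels and chest medallions.
# Mirrored left and right chains each link a "cnst" anchor bone to two segments, all using single-capsule bodies.
# Medallion joints lock twist and have wider swing angles than the tassel joints.
# The tassel anchors carry much larger, offset capsules than the small segments hanging from them.
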PhysicsAsset /BRCosmetics/Characters/Player/Male/Medium/Bodies/M_MED_SamuraiUltra/Meshes/M_MED_Samurai_Ultra_Armor_Dyn_Torso
bodies {
 dyn_back_tassel_01_l [{shapeType=Capsule, length=2.5, radius=1.25, center=<1.9666 0 0>, rotation=<rot -90 0 0>}]
 dyn_back_tassel_01_r [{shapeType=Capsule, length=2.5, radius=1.25, center=<-1.9666 0 0>, rotation=<rot -90 0 0>}]
 dyn_back_tassel_02_l [{shapeType=Capsule, length=2.75, radius=1.25, center=<1.9666 0 0>, rotation=<rot -90 0 0>}]
 dyn_back_tassel_02_r [{shapeType=Capsule, length=2.75, radius=1.25, center=<-1.9666 0 0>, rotation=<rot -90 0 0>}]
 dyn_back_tassel_cnst_l [{shapeType=Capsule, length=11.5, radius=20.5, center=<5.6854 21.8506 0>, rotation=<rot 0 -90 90>}]
 dyn_back_tassel_cnst_r [{shapeType=Capsule, length=11.5, radius=20.5, center=<-5.6854 -21.7331 0>, rotation=<rot 0 -90 90>}]
 dyn_chest_medalion_01_l [{shapeType=Capsule, length=1.6192, radius=1.6188, center=<1.2892 -0.2665 0.0558>, rotation=<rot 90 0 0>}]
 dyn_chest_medalion_01_r [{shapeType=Capsule, length=1.6192, radius=1.6188, center=<-1.2896 0.2663 -0.0558>, rotation=<rot 90 0 0>}]
 dyn_chest_medalion_02_l [{shapeType=Capsule, length=7.1673, radius=2.9778, center=<5.8777 1.2085 -0.1788>, rotation=<rot 90 0 0>}]
 dyn_chest_medalion_02_r [{shapeType=Capsule, length=7.1673, radius=2.9778, center=<-5.8776 -1.2084 0.1788>, rotation=<rot 90 0 0>}]
 dyn_chest_medalion_cnst_l [{shapeType=Capsule, length=0.505, radius=0.505, rotation=<rot 0 -90 0>}]
 dyn_chest_medalion_cnst_r [{shapeType=Capsule, length=0.505, radius=0.505, rotation=<rot 0 -90 0>}]
}
constraints [
 {
  bone1Name=dyn_chest_medalion_02_l
  bone2Name=dyn_chest_medalion_01_l
  swing1LimitDegrees=45
  swing2LimitDegrees=45
  twistLimitDegrees=5
  twistMotion=Locked
 }
 {
  bone1Name=dyn_chest_medalion_02_r
  bone2Name=dyn_chest_medalion_01_r
  swing1LimitDegrees=45
  swing2LimitDegrees=45
  twistLimitDegrees=5
  twistMotion=Locked
 }
 {bone1Name=dyn_back_tassel_01_l, bone2Name=dyn_back_tassel_cnst_l, swing1LimitDegrees=20, swing2LimitDegrees=20, twistLimitDegrees=1}
 {bone1Name=dyn_back_tassel_01_r, bone2Name=dyn_back_tassel_cnst_r, swing1LimitDegrees=20, swing2LimitDegrees=20, twistLimitDegrees=1}
 {
  bone1Name=dyn_chest_medalion_01_l
  bone2Name=dyn_chest_medalion_cnst_l
  swing1LimitDegrees=45
  swing2LimitDegrees=45
  twistLimitDegrees=5
  twistMotion=Locked
 }
 {
  bone1Name=dyn_chest_medalion_01_r
  bone2Name=dyn_chest_medalion_cnst_r
  swing1LimitDegrees=45
  swing2LimitDegrees=45
  twistLimitDegrees=5
  twistMotion=Locked
 }
 {bone1Name=dyn_back_tassel_02_r, bone2Name=dyn_back_tassel_01_r, swing1LimitDegrees=20, swing2LimitDegrees=20, twistLimitDegrees=0}
 {bone1Name=dyn_back_tassel_02_l, bone2Name=dyn_back_tassel_01_l, swing1LimitDegrees=20, swing2LimitDegrees=20, twistLimitDegrees=0}
]
