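# Physics asset for the waist of the male medium Samurai Ultra armor, defining bodies and joints for its dangling hip plates, front and back loincloths, and waist ropes.
# Each piece is a bone chain of mostly capsule bodies rooted at the pelvis, with thigh capsules on free-swinging, free-twisting joints.
# Chain joints allow little twist; hip armor and back cloth swing mainly on one axis, lower front cloth is nearly hinge-like, and ropes swing more widely.
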PhysicsAsset /BRCosmetics/Characters/Player/Male/Medium/Bodies/M_MED_SamuraiUltra/Meshes/M_MED_Samurai_Ultra_Armor_Dyn_Waist
bodies {
 dyn_back_loincloth_01 [{shapeType=Capsule, length=10.231, radius=7.0995, center=<7.3496 -1.7566 0>}]
 dyn_back_loincloth_02 [{shapeType=Capsule, length=11.1128, radius=3.9004, center=<2.6285 0.929 0>}]
 dyn_back_loincloth_03 [{shapeType=Capsule, length=12.0464, radius=5, center=<2.4853 0.3485 -0.0043>}]
 dyn_back_loincloth_04 [{shapeType=Capsule, length=12.8784, radius=7, center=<4.3265 -0.3781 -0.1094>}]
 dyn_front_loincloth_01 [{shapeType=Capsule, length=13.2719, radius=5.2744, center=<9.0502 0.2914 0.9518>}]
 dyn_front_loincloth_02 [{shapeType=Box, extentX=7, extentY=1.5, extentZ=19.5, center=<3.6595 0 0>}]
 dyn_front_loincloth_03 [{shapeType=Capsule, length=12.0464, radius=4.0627, center=<3.8546 -0.2581 -0.0043>}]
 dyn_front_loincloth_04 [{shapeType=Capsule, length=16.75, radius=4, shapeName=dyn_front_loincloth_04_capsule, center=<5.0362 0 0>}]
 dyn_hip_armor_01_l [{shapeType=Capsule, length=15.11, radius=4.566, center=<1.9979 0.5669 0.0112>}]
 dyn_hip_armor_01_r [{shapeType=Capsule, length=17.9827, radius=5.508, center=<-2.5953 -0.3027 -0.0112>}]
 dyn_hip_armor_02_l [{shapeType=Capsule, length=15.11, radius=4.566, center=<2.8063 0.6966 0.2359>}]
 dyn_hip_armor_02_r [{shapeType=Capsule, length=17.6339, radius=5.508, center=<-4.1504 -0.5214 -0.2358>}]
 dyn_hip_armor_03_l [{shapeType=Capsule, length=15.11, radius=4.566, center=<4.2829 0.4559 0.3834>}]
 dyn_hip_armor_03_r [{shapeType=Capsule, length=17.7604, radius=5.508, center=<-4.8804 -0.1835 -0.3834>}]
 dyn_hip_armor_04_l [{shapeType=Capsule, length=15.11, radius=4.566, shapeName=dyn_hip_armor_04_l_capsule, center=<4.0727 0 0>}]
 dyn_hip_armor_04_r [{shapeType=Capsule, length=17.0653, radius=5.508, shapeName=dyn_hip_armor_04_r_capsule, center=<-3.3386 0 0.089>}]
 dyn_hip_armor_mstr_l [{shapeType=Capsule, length=11.2096, radius=5.4744, center=<5.7763 0.2524 0.2212>}]
 dyn_hip_armor_mstr_r [{shapeType=Capsule, length=11.2096, radius=5.4744, center=<-5.7763 0.5291 -0.2212>}]
 dyn_waist_rope_bk_01 [{shapeType=Capsule, length=3, radius=3, center=<4 0 0>, rotation=<rot 90 0 0>}]
 dyn_waist_rope_bk_02 [{shapeType=Capsule, length=3, radius=3, center=<4 0 0>, rotation=<rot 90 0 0>}]
 dyn_waist_rope_fr_01 [{shapeType=Capsule, length=3, radius=3, center=<4 0 0>, rotation=<rot 90 0 0>}]
 dyn_waist_rope_fr_02 [{shapeType=Capsule, length=3, radius=3, center=<4 0 0>, rotation=<rot 90 0 0>}]
 pelvis [{shapeType=Capsule, length=8.9888, radius=7.3892, center=<-8.1286 -4.4234 0>}]
 thigh_l [{shapeType=Capsule, length=26.5374, radius=11.2126, center=<-23.8118 -0.9393 0.2656>, rotation=<rot 71.9239 179.9999 180>}]
 thigh_r [{shapeType=Capsule, length=26.5374, radius=11.2126, center=<24.4772 1.1974 -0.7663>, rotation=<rot 71.9239 179.9999 180>}]
}
constraints [
 {bone1Name=dyn_waist_rope_bk_02, bone2Name=dyn_waist_rope_bk_01, swing1LimitDegrees=45, swing2LimitDegrees=45, twistLimitDegrees=5}
 {bone1Name=dyn_waist_rope_fr_02, bone2Name=dyn_waist_rope_fr_01, swing1LimitDegrees=45, swing2LimitDegrees=45, twistLimitDegrees=5}
 {bone1Name=dyn_waist_rope_bk_01, bone2Name=pelvis, swing1LimitDegrees=45, swing2LimitDegrees=45, twistLimitDegrees=5}
 {bone1Name=dyn_waist_rope_fr_01, bone2Name=pelvis, swing1LimitDegrees=45, swing2LimitDegrees=45, twistLimitDegrees=5}
 {
  bone1Name=thigh_l
  bone2Name=pelvis
  swing1LimitDegrees=45
  swing1Motion=Free
  swing2LimitDegrees=45
  swing2Motion=Free
  twistLimitDegrees=45
  twistMotion=Free
 }
 {
  bone1Name=thigh_r
  bone2Name=pelvis
  swing1LimitDegrees=45
  swing1Motion=Free
  swing2LimitDegrees=45
  swing2Motion=Free
  twistLimitDegrees=45
  twistMotion=Free
 }
 {bone1Name=dyn_hip_armor_mstr_l, bone2Name=pelvis, swing1LimitDegrees=20, swing2LimitDegrees=5, twistLimitDegrees=5}
 {bone1Name=dyn_hip_armor_01_l, bone2Name=dyn_hip_armor_mstr_l, swing1LimitDegrees=20, swing2LimitDegrees=5, twistLimitDegrees=5}
 {bone1Name=dyn_hip_armor_02_l, bone2Name=dyn_hip_armor_01_l, swing1LimitDegrees=20, swing2LimitDegrees=5, twistLimitDegrees=5}
 {bone1Name=dyn_hip_armor_03_l, bone2Name=dyn_hip_armor_02_l, swing1LimitDegrees=20, swing2LimitDegrees=5, twistLimitDegrees=5}
 {bone1Name=dyn_hip_armor_04_l, bone2Name=dyn_hip_armor_03_l, swing1LimitDegrees=20, swing2LimitDegrees=5, twistLimitDegrees=5}
 {bone1Name=dyn_hip_armor_mstr_r, bone2Name=pelvis, swing1LimitDegrees=20, swing2LimitDegrees=5, twistLimitDegrees=5}
 {bone1Name=dyn_hip_armor_01_r, bone2Name=dyn_hip_armor_mstr_r, swing1LimitDegrees=20, swing2LimitDegrees=5, twistLimitDegrees=5}
 {bone1Name=dyn_hip_armor_02_r, bone2Name=dyn_hip_armor_01_r, swing1LimitDegrees=20, swing2LimitDegrees=5, twistLimitDegrees=5}
 {bone1Name=dyn_hip_armor_03_r, bone2Name=dyn_hip_armor_02_r, swing1LimitDegrees=20, swing2LimitDegrees=5, twistLimitDegrees=5}
 {bone1Name=dyn_hip_armor_04_r, bone2Name=dyn_hip_armor_03_r, swing1LimitDegrees=20, swing2LimitDegrees=5, twistLimitDegrees=5}
 {bone1Name=dyn_back_loincloth_01, bone2Name=pelvis, swing1LimitDegrees=20, swing2LimitDegrees=5, twistLimitDegrees=5}
 {bone1Name=dyn_back_loincloth_02, bone2Name=dyn_back_loincloth_01, swing1LimitDegrees=20, swing2LimitDegrees=5, twistLimitDegrees=5}
 {bone1Name=dyn_back_loincloth_03, bone2Name=dyn_back_loincloth_02, swing1LimitDegrees=20, swing2LimitDegrees=5, twistLimitDegrees=5}
 {bone1Name=dyn_back_loincloth_04, bone2Name=dyn_back_loincloth_03, swing1LimitDegrees=20, swing2LimitDegrees=5, twistLimitDegrees=5}
 {bone1Name=dyn_front_loincloth_01, bone2Name=pelvis, swing1LimitDegrees=20, swing2LimitDegrees=10, twistLimitDegrees=5}
 {bone1Name=dyn_front_loincloth_02, bone2Name=dyn_front_loincloth_01, swing1LimitDegrees=45, swing2LimitDegrees=1, twistLimitDegrees=1}
 {bone1Name=dyn_front_loincloth_03, bone2Name=dyn_front_loincloth_02, swing1LimitDegrees=45, swing2LimitDegrees=1, twistLimitDegrees=1}
 {bone1Name=dyn_front_loincloth_04, bone2Name=dyn_front_loincloth_03, swing1LimitDegrees=45, swing2LimitDegrees=1, twistLimitDegrees=1}
]
