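# Animation Blueprint for the male Samurai Ultra armor body, driving its secondary motion on a pose copied from the attached parent mesh.
# Constraints and look-ats position armor bones, then a Control Rig and RigidBody nodes simulate torso, waist and left shoulderpad, with a separate front-end torso asset.
# Skydiving states set a head-to-pelvis-based external force for torso and waist physics.
# On customization load, a backbling check can hide a material element.
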AnimBlueprint /BRCosmetics/Characters/Player/Male/Medium/Bodies/M_MED_SamuraiUltra/Meshes/M_MED_Samurai_Ultra_Armor_AnimBP
target_skeleton=&/M_MED_Samurai_Ultra_Armor_Skeleton
graphs [
 AnimGraph {nodes [
   n1 = "Constraint - dyn_chest_medalion_cnst_l (PPPPP)"!(
    ComponentPose=Local To Component(LocalPose=Copy Pose From Mesh() {bUseAttachedParent=true})
    Alpha=1
    ConstraintWeights_0=0.0271
    ConstraintWeights_1:float=0.1141
    ConstraintWeights_2:float=0.3101
    ConstraintWeights_3:float=0.1317
    ConstraintWeights_4:float=0.4171
   ) -> (Pose) {
    alphaBoolBlend {blendOption=Linear}
    boneToModify {boneName=dyn_chest_medalion_cnst_l}
    constraintSetup [
     {targetBone {boneName=spine_02}}
     {targetBone {boneName=spine_03}}
     {targetBone {boneName=spine_04}}
     {targetBone {boneName=spine_05}}
     {targetBone {boneName=clavicle_l}}
    ]
    constraintWeights [1,1,1,1,1]
   }
   n2 = "Constraint - dyn_chest_medalion_cnst_r (PPPPP)"!(
    ComponentPose=@n1
    Alpha=1
    ConstraintWeights_0=0.4171
    ConstraintWeights_1:float=0.0271
    ConstraintWeights_2:float=0.1141
    ConstraintWeights_3:float=0.3101
    ConstraintWeights_4:float=0.1317
   ) -> (Pose) {
    alphaBoolBlend {blendOption=Linear}
    boneToModify {boneName=dyn_chest_medalion_cnst_r}
    constraintSetup [
     {targetBone {boneName=clavicle_r}}
     {targetBone {boneName=spine_02}}
     {targetBone {boneName=spine_03}}
     {targetBone {boneName=spine_04}}
     {targetBone {boneName=spine_05}}
    ]
    constraintWeights [1,1,1,1,1]
   }
   n3 = "Constraint - dyn_back_tassel_cnst_l (PPPP)"!(
    ComponentPose=@n2
    Alpha=1
    ConstraintWeights_0=0.8041
    ConstraintWeights_1:float=0.0117
    ConstraintWeights_2:float=0.0698
    ConstraintWeights_3:float=0.1143
   ) -> (Pose) {
    alphaBoolBlend {blendOption=Linear}
    boneToModify {boneName=dyn_back_tassel_cnst_l}
    constraintSetup [{targetBone {boneName=spine_03}}, {targetBone {boneName=spine_04}}, {targetBone {boneName=spine_05}}, {targetBone {boneName=clavicle_l}}]
    constraintWeights [1,1,1,1]
   }
   n4 = "Constraint - dyn_back_tassel_cnst_r (PPPP)"!(
    ComponentPose=@n3
    Alpha=1
    ConstraintWeights_0=0.1143
    ConstraintWeights_1:float=0.8041
    ConstraintWeights_2:float=0.0117
    ConstraintWeights_3:float=0.0698
   ) -> (Pose) {
    alphaBoolBlend {blendOption=Linear}
    boneToModify {boneName=dyn_back_tassel_cnst_r}
    constraintSetup [{targetBone {boneName=clavicle_r}}, {targetBone {boneName=spine_03}}, {targetBone {boneName=spine_04}}, {targetBone {boneName=spine_05}}]
    constraintWeights [1,1,1,1]
   }
   n5 = "Constraint - dyn_front_loincloth_01 (RR)"!(ComponentPose=@n4, Alpha=1, ConstraintWeights_0=0.5, ConstraintWeights_1:float=0.5) -> (Pose) {
    alphaBoolBlend {blendOption=Linear}
    boneToModify {boneName=dyn_front_loincloth_01}
    constraintSetup [{transformType=Rotation, targetBone {boneName=thigh_l}}, {transformType=Rotation, targetBone {boneName=thigh_r}}]
    constraintWeights [1,1]
   }
   n6 = "Constraint - dyn_kneepad_l (PPP)"!(ComponentPose=@n5, Alpha=1, ConstraintWeights_0=0.7871, ConstraintWeights_1:float=0.1232, ConstraintWeights_2:float=0.0897) -> (Pose) {
    alphaBoolBlend {blendOption=Linear}
    boneToModify {boneName=dyn_kneepad_l}
    constraintSetup [{targetBone {boneName=calf_l}}, {targetBone {boneName=calf_twist_02_l}}, {targetBone {boneName=thigh_l}}]
    constraintWeights [1,1,1]
   }
   n7 = "Constraint - dyn_kneepad_r (PPP)"!(ComponentPose=@n6, Alpha=1, ConstraintWeights_0=0.0897, ConstraintWeights_1:float=0.1232, ConstraintWeights_2:float=0.7871) -> (Pose) {
    alphaBoolBlend {blendOption=Linear}
    boneToModify {boneName=dyn_kneepad_r}
    constraintSetup [{targetBone {boneName=thigh_r}}, {targetBone {boneName=calf_twist_02_r}}, {targetBone {boneName=calf_r}}]
    constraintWeights [1,1,1]
   }
   n8 = "Constraint - dyn_shoulderArmor_cnst_l (PP)"!(ComponentPose=@n7, Alpha=1, ConstraintWeights_0=0.7779, ConstraintWeights_1:float=0.2221) -> (Pose) {
    alphaBoolBlend {blendOption=Linear}
    boneToModify {boneName=dyn_shoulderArmor_cnst_l}
    constraintSetup [{targetBone {boneName=upperarm_twist_01_l}}, {targetBone {boneName=clavicle_l}}]
    constraintWeights [1,1]
   }
   n9 = "Constraint - dyn_shoulderArmor_cnst_r (PP)"!(ComponentPose=@n8, Alpha=1, ConstraintWeights_0=0.7779, ConstraintWeights_1:float=0.2221) -> (Pose) {
    alphaBoolBlend {blendOption=Linear}
    boneToModify {boneName=dyn_shoulderArmor_cnst_r}
    constraintSetup [{targetBone {boneName=upperarm_twist_01_r}}, {targetBone {boneName=clavicle_r}}]
    constraintWeights [1,1]
   }
   n10 = Look At - Bone: dyn_shoulderArmor_master_l(ComponentPose=@n9, Alpha=1) {
    alphaBoolBlend {blendOption=Linear}
    boneToModify {boneName=dyn_shoulderArmor_master_l}
    lookAtLocation {x=0, z=3}
    lookAtTarget {boneReference {boneName=lowerarm_l}}
    lookAt_Axis {axis=<1 0>}
    lookUp_Axis {axis {x=0, z=1}}
   }
   n11 = Look At - Bone: dyn_shoulderArmor_master_r(ComponentPose=@n10, Alpha=1) {
    alphaBoolBlend {blendOption=Linear}
    boneToModify {boneName=dyn_shoulderArmor_master_r}
    lookAtLocation {x=0, z=-3}
    lookAtTarget {boneReference {boneName=lowerarm_r}}
    lookAt_Axis {axis=<-1 0>}
    lookUp_Axis {axis {x=0, z=1}}
   }
   n12 = Control Rig!(Source=Component To Local(ComponentPose=@n11), Alpha, calf_shape_back_r:double=0) -> (Pose) {
    alphaBoolBlend {blendOption=Linear}
    controlRigAssetReference {blueprintRigClass=&/M_MED_Samurai_Ultra_Armor_CtrlRig.M_MED_Samurai_Ultra_Armor_CtrlRig_C}
   }
   n13 = Get bIsInFrontEnd()
   n14 = Get skydiving vector force() -> (skydiving vector force:Vector)
   n15 = RigidBody!(ComponentPose=Local To Component(LocalPose=@n12), bAlphaBoolEnabled:bool=NOT Boolean(A=@n13), ExternalForce=@n14) -> (Pose) {
    alphaInputType=Bool
    bOverrideWorldGravity=true
    overridePhysicsAsset=&/M_MED_Samurai_Ultra_Armor_Dyn_Torso
    simulationSpace=BaseBoneSpace
    alphaBoolBlend {blendOption=Linear}
    baseBoneRef {boneName=spine_05}
    componentAppliedLinearAccClamp=<1000 1000 1000>
    componentLinearAccScale=<0.4 0.3 0.3>
    componentLinearVelScale=<0.4 0.3 0.3>
    overrideWorldGravity {z=-980}
   }
   n16 = RigidBody(ComponentPose=@n15, bAlphaBoolEnabled=@n13) {
    alphaInputType=Bool
    bOverrideWorldGravity=true
    overridePhysicsAsset=&/M_MED_Samurai_Ultra_Armor_Dyn_Torso_FrontEnd
    simulationSpace=BaseBoneSpace
    alphaBoolBlend {blendOption=Linear}
    baseBoneRef {boneName=spine_05}
    componentAppliedLinearAccClamp=<1000 1000 1000>
    componentLinearAccScale=<0.4 0.3 0.3>
    componentLinearVelScale=<0.4 0.3 0.3>
    overrideWorldGravity {z=-980}
   }
   n17 = RigidBody!(ComponentPose=@n16, Alpha:float, ExternalForce=@n14) -> (Pose) {
    bOverrideWorldGravity=true
    overridePhysicsAsset=&/M_MED_Samurai_Ultra_Armor_Dyn_Waist
    simulationSpace=BaseBoneSpace
    alphaBoolBlend {blendOption=Linear}
    baseBoneRef {boneName=pelvis}
    componentAppliedLinearAccClamp=<3000 3000 3000>
    componentLinearAccScale=<0.4 0.3 0.3>
    componentLinearVelScale=<0.4 0.3 0.3>
    overrideWorldGravity {z=-980}
   }
   n18 = RigidBody(ComponentPose=@n17) {
    bClampLinearTranslationLimitToRefPose=true
    bOverrideWorldGravity=true
    overridePhysicsAsset=&/M_MED_Samurai_Ultra_Armor_Dyn_Shoulderpad_l
    simulationSpace=BaseBoneSpace
    alphaBoolBlend {blendOption=Linear}
    baseBoneRef {boneName=spine_05}
    componentAppliedLinearAccClamp=<1000 1000 1000>
    componentLinearAccScale=<0.4 0.3 0.3>
    componentLinearVelScale=<0.4 0.3 0.3>
    overrideWorldGravity {z=-980}
   }
   Output Pose(Result=Component To Local(ComponentPose=@n18)) {layerGroup=None}
  ]}
 EventGraph {nodes [
   Event BlueprintUpdateAnimation()
   > n1 = Sequence()
   n2 = Branch(execute=@n1.then_1, Condition=Get bCanPlayCustomAnimations() -> (bCanPlayCustomAnimations:bool))
   n3 = Branch(execute=[@n2.else, @n2.then], Condition=Get bIsSkyDiving())
   > n4 = Branch(Condition=Get bIsSkydiveFloating())
   n5 = Get Direction Between Sockets!(
    Component:Skeletal Mesh Component Object Reference=GetOwnerSkeletalMeshComponent()
    SocketOrBoneNameFrom:Name=head
    SocketOrBoneNameTo:Name=pelvis
   ) -> (ReturnValue:Vector) {is=title K2Node_CallFunction}
   > Set skydiving vector force(skydiving vector force:Vector=vector + vector(
     A:Vector=vector * vector(A:Vector=@n5, B:Vector="75,75,75") -> (ReturnValue:Vector) {is=title K2Node_CommutativeAssociativeBinaryOperator}
     B:Vector="0, 0,5000"
    ) -> (ReturnValue:Vector) {is=title K2Node_CommutativeAssociativeBinaryOperator}) -> (Output_Get:Vector)
   n6 = Branch(execute=@n4.else, Condition=Get bIsSkydiveDiveMode())
   > Set skydiving vector force(skydiving vector force=vector + vector(A=vector * vector(A=@n5, B="180,180,200"), B="0, 0,2000"))
   Set skydiving vector force(execute=@n6.else, skydiving vector force=vector * vector(A=@n5, B="-5,-5,-5"))
   Set skydiving vector force(execute=@n3.else, skydiving vector force="0, 0, 0")
   Event OnCharacterCustomizationLoaded()
   n7 = CheckForBackbling()
   > n8 = Branch(Condition=@n7)
   n9 = GetPartSkeletalMeshComponent()
   > n10 = Get Dynamic Material Instance(SkeletalMeshComponent=@n9)
   > n11 = SetScalarParameterValue(self=@n10, ParameterName=Hide Element 01, Value=1)
   Do Once(execute=[@n11.then, @n8.else])
   > Set isWearingBackbling(isWearingBackbling=OR Boolean(A=@n7, B=@n7.bIsBackblingLong))
   "Not Equal (Object)"!(
    A:Object Reference=FindChildCommponentFromAttachSocket!(
     self:Custom Character Part Anim Instance Object Reference
     Parent:Skeletal Mesh Component Object Reference=@n9
     bIncludeAllDescendants:bool=false
     AttachSocket:Name=Cape
    ) -> (ReturnValue:Skeletal Mesh Component Object Reference)
    B:Object Reference
   ) -> (ReturnValue:bool)
   "Not Equal (Object)"(A=FindChildCommponentFromAttachSocket(Parent=@n9, bIncludeAllDescendants=false, AttachSocket=backpack_br))
   Get isWearingBackbling()
  ]}
]
variables [
 floating force {type=Vector}
 isWearingBackbling {type=boolean}
 isWearingCape {type=boolean}
 skydiving vector force {type=Vector}
]
