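# Physics asset of cape collider shapes for the male medium Samurai Ultra armor body, giving a cape volumes to collide against.
# Overlapping capsules cover the pelvis, spine, clavicles and neck, with a sphere on the head.
# Extra capsules sit on dynamic shoulder armor, hip armor and back loincloth bones, so those pieces also act as colliders.
# Arm, leg and foot bodies hold no shapes, and no constraints link any bodies.
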PhysicsAsset /BRCosmetics/Characters/Player/Male/Medium/Bodies/M_MED_SamuraiUltra/Meshes/M_MED_Samurai_Ultra_Armor_CapeColliders
bodies {
 calf_l []
 calf_r []
 clavicle_l [{shapeType=Capsule, length=7.7636, radius=11.1427, center=<9.9642 -0.7741 -1.5923>, rotation=<rot 75.9352 -0.0001 175.5622>}]
 clavicle_r [{shapeType=Capsule, length=7.7636, radius=11.1427, center=<-9.9642 0.7741 1.5923>, rotation=<rot 75.9351 -0.0001 -4.4378>}]
 dyn_back_loincloth_03 [{shapeType=Capsule, length=14.5, radius=15.5, shapeName=dyn_back_loincloth_03_capsule, center=<0 7.613 0>, rotation=<rot 90 180 179.9998>}]
 dyn_hip_armor_03_l [{shapeType=Capsule, length=15, radius=15, shapeName=dyn_hip_armor_03_l_capsule, center=<0 7.9146 0>, rotation=<rot 90 -156.0997 -166.0998>}]
 dyn_hip_armor_03_r [{shapeType=Capsule, length=15, radius=15, shapeName=dyn_hip_armor_03_r_capsule, center=<1.3225 -7.5003 0>, rotation=<rot 90 180 169.9999>}]
 dyn_shoulderArmor_base_l_02 [{
   shapeType=Capsule
   length=19.25
   radius=10.75
   shapeName=dyn_shoulderArmor_base_l_02_capsule
   center=<-4.5125 -0.0001 -4.8274>
   rotation=<rot 90 -90 -90.0014>
  }]
 dyn_shoulderArmor_base_r_02 [{
   shapeType=Capsule
   length=19.25
   radius=10.75
   shapeName=dyn_shoulderArmor_base_r_02_capsule
   center=<3.7078 0 4.1437>
   rotation=<rot 90 -160.5288 -160.5292>
  }]
 foot_l []
 foot_r []
 head [{shapeType=Sphere, radius=17.3125, center=<11.028 -0.174 0>}]
 lowerarm_l []
 lowerarm_r []
 neck_02 [{shapeType=Capsule, length=13.4375, radius=7.3175, center=<-0.2642 -0.3408 0>, rotation=<rot 90 -171.1193 -175.3978>}]
 pelvis [{shapeType=Capsule, length=8.6562, radius=13.5388, center=<-2.6215 -1.3695 0>}]
 spine_01 [{shapeType=Capsule, length=7.9687, radius=15.1638, center=<-2.6215 -3.5977 0>}]
 spine_02 [
  {shapeType=Capsule, length=7.8437, radius=13.695, center=<-2.6215 -2.3828 0.0001>}
  {shapeType=Capsule, length=29.875, radius=11.1013, center=<2.5024 0.4992 3.9825>, rotation=<rot 90 0.0049 0.0049>}
  {shapeType=Capsule, length=29.875, radius=11.1013, center=<2.4704 0.4991 -3.9825>, rotation=<rot 90 -0.0396 -0.0396>}
 ]
 spine_03 [{shapeType=Capsule, length=6.4062, radius=13.5075, center=<-1.953 -2.7457 0.0001>}]
 spine_04 [{shapeType=Capsule, length=11.9062, radius=16.3825, center=<2.1467 -4.1569 0.0002>}]
 spine_05 [
  {shapeType=Capsule, length=16.8437, radius=18.195, center=<-1.7514 -0.3942 0>}
  {shapeType=Capsule, length=14.0312, radius=15.8125, center=<0.7875 3.1143 0>}
 ]
 thigh_l []
 thigh_r []
 upperarm_l []
 upperarm_r []
}
constraints []
